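# Niagara deploy effect for the DiscoButler umbrella glider, playing a comic-book burst on the glider's skeletal mesh after a shared system delay.
# Layered flash sprites, color-cycling Ben-Day dots and a two-layer POW graphic with overshooting scale pop at the sampled bone or socket.
# A ring of colored dots flies under vortex, drag and gravity, using camera-direction dot products to resize each by which side of the ring it occupies.
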NiagaraSystem /BRCosmetics/Gadgets/Assets/VinderTech_GliderChute/Glider_DiscoButler/FX/NS_Umbrella_DiscoButler_Deploy
dependencies {
 usedDataInterfaces [
  &NiagaraDataInterfaceCurve
  &NiagaraDataInterfaceColorCurve
  &NiagaraDataInterfaceSpriteRendererInfo
  &NiagaraDataInterfaceCamera
  &NiagaraDataInterfaceSkeletalMesh
  &NiagaraDataInterfaceArrayFloat
 ]
 usedDynamicInputs [
  &Vector2DFromFloat
  &SimulationPosition
  &UniformRangedVector
 ]
 usedModules [
  &self:SystemSpawnScript.SystemScriptSource.SystemScriptGraph.NiagaraNodeAssignment_1.SetVariables_#1
  &self:SystemSpawnScript.SystemScriptSource.SystemScriptGraph.NiagaraNodeAssignment_2.SetVariables_#2
  &SystemState
  &Timeline
  &InterpolateOverTime
 ]
}
emitters {
 Ben_Day_Dots001 {
  emitterUpdateScript [
   &EmitterState(
    Life Cycle Mode=::NewEnumerator1
    Loop Behavior=::NewEnumerator1
    Loop Delay=&Multiply_Float(A=$System.SystemDelay, B=0.95)
    Loop Duration=1
    UseLoopDelay=true
   )
   &SpawnBurst_Instantaneous(Spawn Group=1, Use Spawn Probability=false, UseLoopCountLimit=false)
   &SpawnBurst_Instantaneous(Spawn Time=0.144, Use Spawn Probability=false, UseLoopCountLimit=false) {module=SpawnBurst_Instantaneous001}
  ]
  particleSpawnScript [
   &InitializeParticle(
    Lifetime=0.36
    Sprite Rotation Mode=::NewEnumerator1
    Sprite Size Mode=::NewEnumerator3
    Uniform Sprite Size=133
    UsePositionOffset=false
    Write Material Random=true
   )
   &CameraOffset(Camera Offset Amount=25)
  ]
  particleUpdateScript [
   &ParticleState()
   &SkeletalMeshLocation(
    Bone / Socket Index=1
    Bone Sampling Mode=::NewEnumerator4
    MaskSpawnGroup=false
    Sampled Position Offset=<0 0 35>
    Skeletal Mesh=$System.Skeletal Mesh
   )
   &DynamicMaterialParameters(
    Index 0 Param 1=&FloatFromCurve(
     FloatCurve=DataInterface() {propertyValues {Curve {keys [
         {arriveTangent=1.5646, interpMode=RCIM_Cubic, leaveTangent=1.5646, tangentMode=RCTM_User, time=0, value=0.05}
         {arriveTangent=0.5667, interpMode=RCIM_Cubic, leaveTangent=0.5667, tangentMode=RCTM_User, time=0.3672, value=0.7363}
         {interpMode=RCIM_Cubic, tangentMode=RCTM_User, time=1, value=1}
        ]}}}
     Scale Curve=0.4
    )
    Index 0 Param 2=&FloatFromCurve(
     FloatCurve=DataInterface() {propertyValues {Curve {keys [
         {arriveTangent=-0.0405, interpMode=RCIM_Cubic, leaveTangent=-0.0405, tangentMode=RCTM_User, time=0.5, value=0.5388}
         {arriveTangent=-0.0719, interpMode=RCIM_Cubic, leaveTangent=-0.0719, tangentMode=RCTM_User, time=1, value=0.2}
        ]}}}
     Scale Curve=0.25
    )
    Module.Index0Param0Write=true
    Module.Index0Param1Write=true
    Module.Index0Param2Write=true
    Module.Index0Param3Write=true
   )
   /SetVariables(
    Particles.BlackColor:LinearColor=&ColorFromCurve(ColorCurve=DataInterface() {propertyValues {
       AlphaCurve {keys [<at 0.7 1 RCIM_Cubic>,<at 0.95 0>]}
       BlueCurve {keys [{arriveTangent=0.9686, interpMode=RCIM_Cubic, time=0.3734, value=0}, <at 1 0.0008>]}
       GreenCurve {keys [{arriveTangent=-1.1049, interpMode=RCIM_Cubic, time=0.3734, value=0}, <at 1 0>]}
       RedCurve {keys [{arriveTangent=-0.6493, interpMode=RCIM_Cubic, time=0.3734, value=0}, <at 1 0.04>]}
      }})
    Particles.BlueColor:LinearColor=&MaskLinearColorBySpawnGroup(
     Masked Linear Color=&ColorFromCurve(ColorCurve=DataInterface() {propertyValues {
        AlphaCurve {keys [<at 0 0>, <at 0.0974 1>, {arriveTangent=-1.1729, interpMode=RCIM_Cubic, leaveTangent=-1.1729, time=0.7, value=1}, <at 0.95 0>]}
        BlueCurve {keys [
          <at 0.2744 0.045 RCIM_Cubic>
          <at 0.3449 0.3631>
          {arriveTangent=1.4152, interpMode=RCIM_Cubic, leaveTangent=1.4152, time=0.3734, value=0.8438}
          <at 0.6183 0.75>
          <at 0.7235 0.0052>
          <at 0.8546 0.0015>
         ]}
        GreenCurve {keys [
          <at 0.2744 0.6513 RCIM_Cubic>
          <at 0.3449 0.6706>
          {arriveTangent=-1.6143, interpMode=RCIM_Cubic, leaveTangent=-1.6143, time=0.3734, value=0.4721}
          <at 0.6183 0.2293>
          <at 0.7235 0.0018>
          <at 0.8546 0>
         ]}
        RedCurve {keys [
          <at 0.2744 1.5 RCIM_Cubic>
          <at 0.3449 0.2068>
          {arriveTangent=-0.9487, interpMode=RCIM_Cubic, leaveTangent=-0.9487, time=0.3734, value=0.0019}
          <at 0.6183 -0.0525>
          <at 0.7235 0>
          <at 0.8546 0.5677>
         ]}
       }})
     Passthrough Linear Color=&ColorFromCurve(ColorCurve=DataInterface() {propertyValues {
        AlphaCurve {keys [<at 0 0>,<at 0.2406 1>,<at 0.7 1 RCIM_Cubic>,<at 0.95 0>]}
        BlueCurve {keys [
          <at 0.0923 0.9478>
          <at 0.2285 0.0435>
          {arriveTangent=-0.0367, interpMode=RCIM_Cubic, leaveTangent=-0.0367, time=0.3389, value=0.045}
          {interpMode=RCIM_Cubic, leaveTangent=0.3824, time=0.6769, value=0.027}
         ]}
        GreenCurve {keys [
          <at 0.0923 0.2373>
          <at 0.2285 0.6>
          {arriveTangent=-0.4823, interpMode=RCIM_Cubic, leaveTangent=-0.4823, time=0.3389, value=0.6513}
          {interpMode=RCIM_Cubic, leaveTangent=-0.4747, time=0.6769, value=0.3837}
         ]}
        RedCurve {keys [
          <at 0.0923 0>
          <at 0.2285 0.2377>
          {arriveTangent=1.3515, interpMode=RCIM_Cubic, leaveTangent=1.3515, time=0.3389, value=1.5}
          {interpMode=RCIM_Cubic, leaveTangent=-0.7728, time=0.6769, value=0.8438}
         ]}
       }})
     Spawn Group Mask=1
    )
   ) {isSetParameters=true}
  ]
  rendererClasses [&NiagaraSpriteRendererProperties]
  renderers [
   {rendererClass=&NiagaraSpriteRendererProperties}
   {rendererClass=&NiagaraSpriteRendererProperties, rendererIndex=1}
  ]
 }
 Dots_BackShifting {
  emitterUpdateScript [
   &EmitterState(Delay First Loop Only=true, Life Cycle Mode=::NewEnumerator1, Loop Delay=$System.SystemDelay, Loop Duration=0.34, UseLoopDelay=true)
   &SpawnBurst_Instantaneous(Spawn Count=12, Use Spawn Probability=false, UseLoopCountLimit=false)
   &SpawnBurst_Instantaneous(Spawn Count=12, Spawn Group=1, Spawn Time=0.17, Use Spawn Probability=false, UseLoopCountLimit=false) {module=SpawnBurst_Instantaneous001}
   /SetVariables(
    Emitter.EmiCameraForward:Vector3f=&ReturnCameraVector(Camera Vectors=::NewEnumerator1, Coordinate Space=::World)
    Emitter.LocalUpDot:NiagaraFloat=&FloatFromCurve(
     FloatCurve=DataInterface() {propertyValues {Curve {keys [
         <at 0 1>
         {arriveTangent=-4.2817, interpMode=RCIM_Cubic, leaveTangent=-4.2817, tangentMode=RCTM_User, time=0.2033, value=1}
         {arriveTangent=-4.1638, leaveTangent=-1.3768, tangentMode=RCTM_User, time=0.7302, value=0}
         <at 1 0 RCIM_Cubic>
        ]}}}
     CurveIndex=&DotProduct(Vector A=<0 0 1>, Vector B=&ReturnCameraVector(Camera Vectors=::NewEnumerator1, Coordinate Space=::Local))
    )
   ) {isSetParameters=true}
  ]
  particleSpawnScript [
   &InitializeParticle(
    Color Mode=::NewEnumerator0
    Lifetime Max=0.82
    Lifetime Min=0.72
    Lifetime Mode=::NewEnumerator1
    Sprite Size Mode=::NewEnumerator1
    Uniform Sprite Size Max=5
    Uniform Sprite Size Min=4.5
    UsePositionOffset=false
    Write Material Random=true
   )
   &SkeletalMeshLocation(Bone / Socket Index=1, Bone Sampling Mode=::NewEnumerator4, MaskSpawnGroup=false, Skeletal Mesh=$System.Skeletal Mesh)
   &ShapeLocation(
    MaskSpawnGroup=false
    Offset=<0 0 -20>
    Offset Coordinate Space=::Local
    Offset Mode=::NewEnumerator0
    Override Seed=false
    Ring / Disc Distribution Mode:ENiagaraSphereDistributionMode=::NewEnumerator1
    Ring Radius=72
    Shape Primitive=::NewEnumerator4
    U Position=&Add_Float(
     A=&Multiply_FloatByInt(Float=&Divide_Float(A=1, B=12), Integer=&ReturnParticleIDAttributes())
     B=&Divide_Float(A=&MaskFloatBySpawnGroup(Masked Float=1, Passthrough Float=1.5), B=12)
    )
    UseOwnerScale=true
   )
   /SetVariables(
    Particles.BlackDot:NiagaraBool=&SetBoolByFloatComparison(A=&RandomRangeFloat(Minimum=0, Maximum=1), B=0.9)
    Particles.CirclePosition:Vector3f=&TransformVector(
     Vector=&NormalizeVector(
      Vector To Normalize=&SubtractVectorFromPosition(
       Position=$Particles.Position
       Vector=&AddVectorToPosition(Position=$Particles.Initial.SkeletalMeshLocation.SampledPosition, Vector=<0 0 -20>)
      )
      Safe Normalize=true
      Safe Fallback Threshold=0
      Safe Fallback Vector=<0 0 0.001>
     )
     Source Space=::Simulation
     Destination Space=::Simulation
    )
   ) {isSetParameters=true}
   &VortexForce(
    UseOriginOffset=false
    Vortex Force Amount=&Multiply_Float(A=200, B=$System.InterpolateOverTime001.InterpolateOverTime_Float.Moving Average)
   )
   &ApplyInitialForces()
  ]
  particleUpdateScript [
   &ParticleState()
   /SetVariables(
    Particles.DotBody:NiagaraFloat=&OneMinusFloat(Float=&Multiply_Float(
      A=&FloatFromCurve(
       FloatCurve=DataInterface() {propertyValues {Curve {keys [
           {arriveTangent=1.9877, time=0, value=0}
           {arriveTangent=47.1252, interpMode=RCIM_Cubic, leaveTangent=47.1252, tangentMode=RCTM_User, time=0.9771, value=0}
           <at 1 1>
          ]}}}
       CurveIndex=&RemapRange(
        Input Value To Remap=&DotProduct(Vector A=$Particles.Initial.CirclePosition, Vector B=$Emitter.EmiCameraForward)
        Input Min=-1
        Input Max=1
        Output Min=1
        Output Max=0
       )
       Scale Curve=&OneMinusFloat(Float=&Abs_Float(Float:NiagaraFloat=$System.InterpolateOverTime001.InterpolateOverTime_Float.Moving Average))
      )
      B=&FloatFromCurve(FloatCurve=DataInterface() {propertyValues {Curve {keys [<at 0.6687 0>,<at 0.8 1>]}}})
     ))
    Particles.DotFacing:NiagaraFloat=&FloatFromCurve(
     FloatCurve=DataInterface() {propertyValues {Curve {keys [
         <at 0 0>
         {arriveTangent=17.4655, interpMode=RCIM_Cubic, leaveTangent=17.4655, tangentMode=RCTM_User, time=0.4969, value=0}
         {arriveTangent=1.9877, leaveTangent=1.9877, time=0.5899, value=1}
         {arriveTangent=-20.5781, leaveTangent=-20.5781, time=1, value=1}
        ]}}}
     CurveIndex=&RemapRange(
      Input Value To Remap=&DotProduct(Vector A=$Particles.Initial.CirclePosition, Vector B=$Emitter.EmiCameraForward)
      Input Min=-1
      Input Max=1
      Output Min=1
      Output Max=0
     )
    )
   ) {isSetParameters=true}
   &ScaleSpriteSize(
    Scale Sprite Size Mode=::NewEnumerator0
    Uniform Scale Factor=&Multiply_Float(
     A=&FloatFromCurve(
      FloatCurve=DataInterface() {propertyValues {Curve {keys [
          {arriveTangent=2.4711, time=0, value=1}
          {arriveTangent=0.4466, time=0.0807, value=1.2235}
          {arriveTangent=-0.4864, leaveTangent=-0.4864, time=0.8774, value=1}
          {arriveTangent=-8.4195, leaveTangent=-8.4195, time=1, value=0}
         ]}}}
      CurveIndex=&Add_Float(
       A=&Multiply_Float(A=&OneMinusFloat(Float=$Particles.DotFacing), B=$Emitter.LocalUpDot)
       B=$Particles.NormalizedAge
      )
     )
     B=&FloatFromCurve(
      FloatCurve=DataInterface() {propertyValues {Curve {keys [
          {arriveTangent=2.4711, time=0, value=0.6}
          {arriveTangent=0.4466, leaveTangent=0.4466, time=0.0807, value=1.0079}
          {arriveTangent=-0.4864, leaveTangent=-0.4864, time=1, value=1}
         ]}}}
      CurveIndex=$Particles.Age
     )
    )
   ) {module=ScaleSpriteSize001}
   &ScaleSpriteSize(
    Scale Sprite Size Mode=::NewEnumerator0
    Uniform Scale Factor=&Lerp_Float(A=1, B=$Particles.DotBody, Alpha=$Particles.DotFacing)
   ) {module=ScaleSpriteSize002}
   &InheritSourceMovement(
    Applied Position Scale=&Multiply_VectorByFloat(
     Vector=<1 1 1>
     Float=&FloatFromCurve(FloatCurve=DataInterface() {propertyValues {Curve {keys [<at 0 1>,<at 1 0.95>]}}})
    )
    ApplyForce=false
    ApplyPosition=true
    ApplyPositionFalloff=false
   )
   &Drag(Drag=2, Use Linear Drag=true, Use Rotational Drag=true)
   &GravityForce(Gravity=&MakeVector(X=0, Y=0, Z=-450))
   &SolveForcesAndVelocity(Clamp Velocity=false, Limit Acceleration=false)
   /SetVariables(
    Particles.ColBack:LinearColor=<rgba 0 0 0 1>
    Particles.ColFront:LinearColor=&MakeCustomLinearColorFromBool(
     Bool=$Particles.Initial.BlackDot
     True Color=&ColorFromCurve(
      ColorCurve=DataInterface() {propertyValues {
        AlphaCurve {keys [<at 0 1 RCIM_Constant>]}
        BlueCurve {keys [<at 0 0.0578 RCIM_Constant>,<at 0.2838 0.0061 RCIM_Constant>,<at 0.6836 0.1515 RCIM_Constant>,<at 0.9611 0>]}
        GreenCurve {keys [<at 0 0.0048 RCIM_Constant>,<at 0.2838 0.2908 RCIM_Constant>,<at 0.6836 0.5469 RCIM_Constant>,<at 0.9611 0.6236>]}
        RedCurve {keys [<at 0 0.4287 RCIM_Constant>,<at 0.2838 0.6094 RCIM_Constant>,<at 0.6836 0 RCIM_Constant>,<at 0.9611 0.7043>]}
       }}
      CurveIndex=&RandomRangeFloat(Minimum=0, Maximum=1)
     )
     False Color=&ColorFromCurve(
      ColorCurve=DataInterface() {propertyValues {
        AlphaCurve {keys [<at 0 1 RCIM_Constant>]}
        BlueCurve {keys [<at 0 0.5913 RCIM_Constant>,<at 0.2792 0.9739 RCIM_Constant>,<at 0.8753 0.6584 RCIM_Constant>]}
        GreenCurve {keys [<at 0 0.3506 RCIM_Constant>,<at 0.2792 0.1136 RCIM_Constant>,<at 0.8753 0.0782 RCIM_Constant>]}
        RedCurve {keys [<at 0 0 RCIM_Constant>,<at 0.2792 0 RCIM_Constant>,<at 0.8753 0 RCIM_Constant>]}
       }}
      CurveIndex=&RandomRangeFloat(Minimum=0, Maximum=1)
     )
    )
    Particles.SizeBack:Vector2f=&Multiply_Vector2D_ByFloat(Float=1.08, Vector2D=$Particles.SpriteSize)
   ) {isSetParameters=true}
  ]
  rendererClasses [&NiagaraSpriteRendererProperties]
  renderers [
   {rendererClass=&NiagaraSpriteRendererProperties}
   {rendererClass=&NiagaraSpriteRendererProperties, rendererIndex=1}
  ]
 }
 Flash_Elements {
  emitterUpdateScript [
   &EmitterState(
    Enable Distance Culling=true
    Life Cycle Mode=::NewEnumerator1
    Loop Behavior=::NewEnumerator1
    Loop Delay=$System.SystemDelay
    Loop Duration=1
    Max Distance Response=::KillImmediately
    Module.bUseMaxDistance=true
    Module.bUseMinDistance=false
    Scalability Mode=::NewEnumerator1
    UseLoopDelay=true
   )
   &SpawnBurst_Instantaneous(Spawn Count=2, Spawn Group=2, Use Spawn Probability=false, UseLoopCountLimit=false) {module=SpawnBurst_Instantaneous003}
   &SpawnBurst_Instantaneous(Spawn Count=1, Spawn Group=1, Use Spawn Probability=false) {module=SpawnBurst_Instantaneous002}
   &SpawnBurst_Instantaneous(Use Spawn Probability=false, UseLoopCountLimit=false)
  ]
  particleSpawnScript [
   &InitializeParticle(
    Lifetime=&MergeHelper_FloatSwitch(0=0.09, 1=0.2, 2=0.45, 3=?, Index=$Emitter.SpawnGroup, Num=::NewEnumerator1)
    Sprite Rotation Mode=::NewEnumerator2
    Sprite Size Mode=::NewEnumerator3
    Uniform Sprite Size=&Multiply_Float(
     A=&MergeHelper_FloatSwitch(
      0=&UniformRangedFloat(Minimum=150, Maximum=170)
      1=398.0062
      2=&Multiply_Float(A=180, B=&RandomRangeFloat(Minimum=0.4, Maximum=1))
      3=?
      Index=$Emitter.SpawnGroup
      Num=::NewEnumerator1
     )
     B=1
    )
    UsePositionOffset=false
    Write Material Random=true
   )
   &CameraOffset(Camera Offset Amount=&MergeHelper_FloatSwitch(0=10, 1=25, 2=10, 3=?, Index=$Emitter.SpawnGroup, Num=::NewEnumerator1))
   /SetVariables(Particles.DynamicParameterAlt2:Vector4f=&MakeVector4(X=1, Y=1, Z=0, W=0)) {isSetParameters=true}
   /SetVariables() {isSetParameters=true}
   /SetVariables(Particles.VisibilityTag=$Emitter.SpawnGroup) {isSetParameters=true}
  ]
  particleUpdateScript [
   &ParticleState()
   &SkeletalMeshLocation(
    Bone / Socket Index=1
    Bone Sampling Mode=::NewEnumerator4
    MaskSpawnGroup=false
    Sampled Position Offset=<0 0 35>
    Skeletal Mesh=$System.Skeletal Mesh
   )
   &ScaleSpriteSize(
    Scale Sprite Size Mode=::NewEnumerator0
    Uniform Scale Factor=&MergeHelper_FloatSwitch(
     0=&FloatFromCurve(FloatCurve=DataInterface() {propertyValues {Curve {keys [
          {arriveTangent=2.105, interpMode=RCIM_Cubic, leaveTangent=2.105, tangentMode=RCTM_User, time=0, value=0}
          {arriveTangent=0.0224, interpMode=RCIM_Cubic, leaveTangent=0.0224, tangentMode=RCTM_User, time=1, value=1}
         ]}}})
     1=&FloatFromCurve(FloatCurve=DataInterface() {propertyValues {Curve {keys [
          {arriveTangent=1.6131, interpMode=RCIM_Cubic, leaveTangent=1.6131, tangentMode=RCTM_User, time=0, value=0.7}
          {arriveTangent=0.0224, interpMode=RCIM_Cubic, leaveTangent=0.0224, tangentMode=RCTM_User, time=1, value=0}
         ]}}})
     2=&FloatFromCurve(FloatCurve=DataInterface() {propertyValues {Curve {keys [{arriveTangent=1.2663, interpMode=RCIM_Cubic, leaveTangent=1.2663, tangentMode=RCTM_User, time=0, value=1.0706}, <at 1 1.4 RCIM_Cubic>]}}})
     3=?
     Index=$Particles.VisibilityTag
     Num=::NewEnumerator1
    )
   ) {module=ScaleSpriteSize001}
   &ScaleColor(
    Scale Alpha=&MergeHelper_FloatSwitch(
     0=&FloatFromCurve(FloatCurve=DataInterface() {propertyValues {Curve {keys [
          {arriveTangent=9.0725, interpMode=RCIM_Cubic, leaveTangent=9.0725, tangentMode=RCTM_User, time=0, value=0.5702}
          {arriveTangent=-1.2928, interpMode=RCIM_Cubic, leaveTangent=-1.2928, tangentMode=RCTM_User, time=0.0555, value=1}
          {arriveTangent=-0.0405, interpMode=RCIM_Cubic, leaveTangent=-0.0405, tangentMode=RCTM_User, time=1, value=0}
         ]}}})
     1=1
     2=&FloatFromCurve(
      FloatCurve=DataInterface() {propertyValues {Curve {keys [
          {arriveTangent=-0.2921, interpMode=RCIM_Cubic, leaveTangent=-0.2921, tangentMode=RCTM_User, time=0, value=7}
          {arriveTangent=-7, interpMode=RCIM_Cubic, leaveTangent=-7, time=0.4, value=4}
          <at 1 0 RCIM_Cubic>
         ]}}}
      Scale Curve=2
     )
     3=?
     Index=$Particles.VisibilityTag
     Num=::NewEnumerator1
    )
    Scale RGB=&MergeHelper_VectorSwitch(
     0=&MakeVectorFromLinearColorRGB(Color=<rgba 2.2534 4.8087 9 1>)
     1=&MakeVectorFromLinearColorRGB(Color=&ColorFromCurve(ColorCurve=DataInterface() {propertyValues {
         AlphaCurve {keys [<at 0 1 RCIM_Cubic>]}
         BlueCurve {keys [<at 0 0.1796>,<at 0.1676 1>,<at 0.338 0.9986>,<at 0.9958 0.5>]}
         GreenCurve {keys [<at 0 1.26>,<at 0.1676 0.5107>,<at 0.338 0.5094>,<at 0.9958 0.024>]}
         RedCurve {keys [<at 0 2>,<at 0.1676 0.0773>,<at 0.338 0.0771>,<at 0.9958 0>]}
        }}))
     2=<1 1 1>
     3=?
     Index=$Particles.VisibilityTag
     Num=::NewEnumerator1
    )
    ScaleA=true
    ScaleRGB=true
   )
   /SetVariables(
    Particles.InnerDynamicParameter:Vector4f=&MakeVector4_FromFloat(Float=&Lerp_Float(A=0, B=0.5, Alpha=$Particles.NormalizedAge))
    Particles.MiddleDynamicParameter:Vector4f=&MakeVector4_FromFloat(Float=&Lerp_Float(A=0.25, B=0.75, Alpha=$Particles.NormalizedAge))
    Particles.OuterDynamicParameter:Vector4f=&MakeVector4_FromFloat(Float=&Lerp_Float(A=0, B=1, Alpha=$Particles.NormalizedAge))
    Particles.SubImageIndex=&RandomRangeFloat(Minimum=0, Maximum=4)
    Particles.SubImageIndexAlt:NiagaraFloat=&RandomRangeFloat(Minimum=0, Maximum=3)
   ) {isSetParameters=true}
   /SetVariables(Particles.DynamicParameterAlt3:Vector4f=&MakeVector4(
     X=&FloatFromCurve(FloatCurve=DataInterface() {propertyValues {Curve {keys [<at 0 1 RCIM_Cubic>, {arriveTangent=-2.742, interpMode=RCIM_Cubic, leaveTangent=-2.742, tangentMode=RCTM_User, time=1, value=0}]}}})
     Y=&FloatFromCurve(FloatCurve=DataInterface() {propertyValues {Curve {keys [
          {arriveTangent=1.8371, interpMode=RCIM_Cubic, leaveTangent=1.8371, tangentMode=RCTM_User, time=0, value=0}
          {arriveTangent=1.6535, interpMode=RCIM_Cubic, leaveTangent=1.6535, tangentMode=RCTM_User, time=1, value=1}
         ]}}})
     Z=0
     W=0
    )) {isSetParameters=true}
  ]
  rendererClasses [&NiagaraSpriteRendererProperties]
  renderers [
   {rendererClass=&NiagaraSpriteRendererProperties}
   {rendererClass=&NiagaraSpriteRendererProperties, rendererIndex=1}
   {rendererClass=&NiagaraSpriteRendererProperties, rendererIndex=2}
  ]
 }
 POW {
  emitterUpdateScript [
   &EmitterState(Life Cycle Mode=::NewEnumerator1, Loop Behavior=::NewEnumerator1, Loop Delay=$System.SystemDelay, Loop Duration=1, UseLoopDelay=true)
   &SpawnBurst_Instantaneous(Spawn Time=0.02, Use Spawn Probability=false, UseLoopCountLimit=false)
   &SpawnBurst_Instantaneous(Spawn Group=1, Spawn Time=0.02, Use Spawn Probability=false, UseLoopCountLimit=false) {module=SpawnBurst_Instantaneous001}
  ]
  particleSpawnScript [
   &InitializeParticle(
    Color=<rgba 1 1 1 2>
    Lifetime=0.476
    Sprite Size=<133 66.5>
    Sprite Size Mode=::NewEnumerator4
    UsePositionOffset=false
    Write Material Random=true
   ) {module=InitializeParticle001}
   &CameraOffset(Camera Offset Amount=5)
   /Game/Effects/Niagara/Modules/Sprite/FN_SpriteUV() {enabled=false}
   &SubUVAnimation(
    Direct Frame Index:NiagaraFloat=&MaskFloatBySpawnGroup(Masked Float=0, Passthrough Float=1)
    "Sprite Renderer "=DataInterface() {propertyValues {SpriteRenderer=&self:POW_0.NiagaraSpriteRendererProperties_0}}
    SubUV Animation Mode:ENiagara_SubUVLookupModeV2=::NewEnumerator5
   )
   /SetVariables(Particles.VisibilityTag=&SpawnGroup()) {isSetParameters=true}
   /SetVariables(Particles.PowSort:NiagaraFloat=&MaskFloatBySpawnGroup(Masked Float=0, Passthrough Float=1)) {enabled=false, isSetParameters=true}
  ]
  particleUpdateScript [
   &ParticleState()
   /SetVariables(
    Particles.ScaleBackground:NiagaraFloat=&FloatFromCurve(FloatCurve=DataInterface() {propertyValues {Curve {keys [
         {arriveTangent=0.7163, interpMode=RCIM_Cubic, leaveTangent=0.7163, tangentMode=RCTM_User, time=0, value=0.5388}
         {arriveTangent=4.6558, interpMode=RCIM_Cubic, leaveTangent=4.6558, time=0.0954, value=0.6669}
         {arriveTangent=12.7184, interpMode=RCIM_Cubic, leaveTangent=-8.363, tangentMode=RCTM_Break, time=0.1737, value=1.3475}
         {arriveTangent=-0.3519, interpMode=RCIM_Cubic, leaveTangent=-0.3519, tangentMode=RCTM_User, time=0.3332, value=0.9576}
         {arriveTangent=0.1052, interpMode=RCIM_Cubic, leaveTangent=0.1052, time=0.5021, value=1.0082}
         <at 0.68 0.9941>
         <at 1 1 RCIM_Cubic>
        ]}}})
    Particles.ScaleBoth:NiagaraFloat=&FloatFromCurve(FloatCurve=DataInterface() {propertyValues {Curve {keys [
         <at 0 0.3 RCIM_Cubic>
         <at 0.25 1>
         {arriveTangent=0.0141, interpMode=RCIM_Cubic, leaveTangent=0.0141, tangentMode=RCTM_User, time=0.481, value=1}
         {arriveTangent=-5.4503, tangentMode=RCTM_User, time=1, value=0}
        ]}}})
    Particles.ScaleForeground:NiagaraFloat=&FloatFromCurve(FloatCurve=DataInterface() {propertyValues {Curve {keys [
         {arriveTangent=8.5234, interpMode=RCIM_Cubic, leaveTangent=8.5234, tangentMode=RCTM_User, time=0, value=0.2}
         {arriveTangent=12.7184, interpMode=RCIM_Cubic, leaveTangent=-3.7344, tangentMode=RCTM_Break, time=0.0799, value=1.2306}
         {arriveTangent=-0.4432, leaveTangent=-0.4432, time=0.5659, value=1}
         <at 1 1 RCIM_Cubic>
        ]}}})
   ) {isSetParameters=true}
   &ScaleSpriteSize(
    Scale Sprite Size Mode=::NewEnumerator0
    Uniform Scale Factor=&Multiply_Float(
     A=&MaskFloatBySpawnGroup(Masked Float=$Particles.ScaleForeground, Passthrough Float=$Particles.ScaleBackground)
     B=$Particles.ScaleBoth
    )
   )
   &ScaleSpriteSize(
    Scale Sprite Size Mode=::NewEnumerator0
    Uniform Scale Factor=&FloatFromCurve(FloatCurve=DataInterface() {propertyValues {Curve {keys [
         <at 0 0 RCIM_Cubic>
         <at 0.36 1>
         {arriveTangent=0.0141, interpMode=RCIM_Cubic, leaveTangent=0.0141, tangentMode=RCTM_User, time=0.481, value=1}
        ]}}})
   ) {enabled=false, module=ScaleSpriteSize001}
   &SkeletalMeshLocation(
    Bone / Socket Index=1
    Bone Sampling Mode=::NewEnumerator4
    MaskSpawnGroup=false
    Sampled Position Offset=<0 0 35>
    Skeletal Mesh=$System.Skeletal Mesh
   )
   &DynamicMaterialParameters(
    Index 0 Param 1=472.6683
    Index 0 Param 2=0.001
    Module.Index0Float0Write=true
    Module.Index0Float1Write=true
    Module.Index0Float2Write=true
    Module.Index0Float3Write=true
   )
  ]
  rendererClasses [&NiagaraSpriteRendererProperties]
  renderers [
   {rendererClass=&NiagaraSpriteRendererProperties}
   {rendererClass=&NiagaraSpriteRendererProperties, rendererIndex=1}
  ]
 }
}
summary {
 systemName=NS_Umbrella_DiscoButler_Deploy
 emitters [
  {emitterName=Ben_Day_Dots001, rendererClasses [&NiagaraSpriteRendererProperties]}
  {emitterName=Flash_Elements, rendererClasses [&NiagaraSpriteRendererProperties]}
  {emitterName=POW, rendererClasses [&NiagaraSpriteRendererProperties]}
  {emitterName=Dots_BackShifting, rendererClasses [&NiagaraSpriteRendererProperties]}
 ]
 userVariables [
  User.bIsFrontEnd {type=NiagaraBool}
  User.bIsFrontEndPreview {type=NiagaraBool}
  User.ForwardDot {type=NiagaraFloat, defaultValue=0, description=""}
  User.RightDot {type=NiagaraFloat, defaultValue=0, description=""}
 ]
}
